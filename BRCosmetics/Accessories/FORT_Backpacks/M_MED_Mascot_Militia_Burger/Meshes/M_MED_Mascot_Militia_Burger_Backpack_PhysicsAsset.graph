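# Physics asset for the Mascot Militia Burger backpack cosmetic, covering its pack body and dangling bottles and straps.
# The pack bone carries two capsules, while left and right bottles use capsules and left and right straps use boxes.
# Each dangling bone is constrained to the pack: bottles swing and twist within modest limits, and straps hinge mostly on one swing axis.
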PhysicsAsset /BRCosmetics/Accessories/FORT_Backpacks/M_MED_Mascot_Militia_Burger/Meshes/M_MED_Mascot_Militia_Burger_Backpack_PhysicsAsset
bodies {
 dyn_bottle_l [{shapeType=Capsule, length=10.2305, radius=3.4285, center=<4 0 0>, rotation=<rot 90 0 0>}]
 dyn_bottle_r [{shapeType=Capsule, length=10.2305, radius=3.4285, center=<4 0 0>, rotation=<rot 90 0 0>}]
 dyn_strap_l [{shapeType=Box, extentX=3.8883, extentY=1.9544, extentZ=3.7759, center=<1.5312 0.3898 0.0107>}]
 dyn_strap_r [{shapeType=Box, extentX=3.8883, extentY=1.9544, extentZ=3.7759, center=<1.5311 0.3898 -0.0107>}]
 pack [
  {shapeType=Capsule, length=20.8059, radius=11.8256, center=<0.2286 -8.6928 -9.7275>, rotation=<rot -2.0133 -90 1.2719>}
  {shapeType=Capsule, length=20.8059, radius=5.2006, center=<0 0 -21.8503>, rotation=<rot 90 0 0>}
 ]
}
constraints [
 {bone1Name=dyn_bottle_l, bone2Name=pack, swing1LimitDegrees=14, swing2LimitDegrees=10, twistLimitDegrees=10}
 {bone1Name=dyn_bottle_r, bone2Name=pack, swing1LimitDegrees=14, swing2LimitDegrees=10, twistLimitDegrees=10}
 {bone1Name=dyn_strap_l, bone2Name=pack, swing1LimitDegrees=50, swing2LimitDegrees=2, twistLimitDegrees=1}
 {bone1Name=dyn_strap_r, bone2Name=pack, swing1LimitDegrees=50, swing2LimitDegrees=2, twistLimitDegrees=1}
]
